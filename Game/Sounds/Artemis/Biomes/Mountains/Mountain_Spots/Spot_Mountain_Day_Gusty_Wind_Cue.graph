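# Outdoor ambient one-shot cue for daytime mountain gusty wind spots, attenuated as a wind tile ambient.
# A quality node sends its first two levels to a random pick between grassland rustle layered with either alpine quad one-shots or six mountain gusts.
# The last two levels use a lighter fixed mix of three gusts and alpine ST one-shots.
# Each layer randomly selects a wave and varies its pitch and volume through a modulator.
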SoundCue /Game/Sounds/Artemis/Biomes/Mountains/Mountain_Spots/Spot_Mountain_Day_Gusty_Wind_Cue
root=QualityLevel_1
nodes [
 {id=Mixer_1, children [Modulator_2,Modulator_0], properties {inputVolume [1,1]}}
 {id=Mixer_2, children [Modulator_3,Modulator_4], properties {inputVolume [1,1]}}
 {id=Mixer_3, children [Modulator_4,Modulator_1], properties {inputVolume [1,1]}}
 {id=Modulator_0, children [Random_3], properties {pitchMax=1.1, pitchMin=0.9, volumeMax=0.9, volumeMin=0.8}}
 {id=Modulator_1, children [Random_0], properties {pitchMax=1.1, volumeMax=1.1, volumeMin=0.9}}
 {id=Modulator_2, children [Random_1], properties {pitchMax=1.1, volumeMax=1.1, volumeMin=0.9}}
 {id=Modulator_3, children [Random_2], properties {pitchMax=1.1, pitchMin=0.9, volumeMax=1.1, volumeMin=0.9}}
 {id=Modulator_4, children [Random_4], properties {pitchMax=1.1, pitchMin=0.9, volumeMax=1.1, volumeMin=0.9}}
 {id=QualityLevel_1, children [Random_5,Random_5,Mixer_1,Mixer_1]}
 {id=Random_0, children [WavePlayer_0,WavePlayer_1,WavePlayer_2,WavePlayer_3,WavePlayer_4,WavePlayer_5], properties {weights [1,1,1,1,1,1]}}
 {id=Random_1, children [WavePlayer_6,WavePlayer_7,WavePlayer_8], properties {weights [1,1,1]}}
 {id=Random_2, children [WavePlayer_9, WavePlayer_10, WavePlayer_11, WavePlayer_12], properties {weights [1,1,1,1]}}
 {id=Random_3, children [WavePlayer_13, WavePlayer_14, WavePlayer_15, WavePlayer_16], properties {weights [1,1,1,1]}}
 {
  id=Random_4
  children [WavePlayer_17, WavePlayer_18, WavePlayer_19, WavePlayer_21, WavePlayer_22, WavePlayer_23, WavePlayer_24, WavePlayer_25, WavePlayer_26]
  properties {weights [1,1,1,1,1,1,1,1,1]}
 }
 {id=Random_5, children [Mixer_2,Mixer_3], properties {weights [1,1]}}
 {id=WavePlayer_0, properties {soundWaveAssetPtr=&/Spot_Mountain_Day_Gusty_Wind_06}}
 {id=WavePlayer_1, properties {soundWaveAssetPtr=&/Spot_Mountain_Day_Gusty_Wind_01}}
 {id=WavePlayer_10, properties {soundWaveAssetPtr=/Game/Sounds/Apollo/Biomes/Alpine/Biome_Alpine_Oneshot_Day_Quad_02}}
 {id=WavePlayer_11, properties {soundWaveAssetPtr=/Game/Sounds/Apollo/Biomes/Alpine/Biome_Alpine_Oneshot_Day_Quad_03}}
 {id=WavePlayer_12, properties {soundWaveAssetPtr=/Game/Sounds/Apollo/Biomes/Alpine/Biome_Alpine_Oneshot_Day_Quad_04}}
 {id=WavePlayer_13, properties {soundWaveAssetPtr=/Game/Sounds/Apollo/Biomes/Alpine/Biome_Alpine_Oneshot_Day_ST_01}}
 {id=WavePlayer_14, properties {soundWaveAssetPtr=/Game/Sounds/Apollo/Biomes/Alpine/Biome_Alpine_Oneshot_Day_ST_02}}
 {id=WavePlayer_15, properties {soundWaveAssetPtr=/Game/Sounds/Apollo/Biomes/Alpine/Biome_Alpine_Oneshot_Day_ST_03}}
 {id=WavePlayer_16, properties {soundWaveAssetPtr=/Game/Sounds/Apollo/Biomes/Alpine/Biome_Alpine_Oneshot_Day_ST_04}}
 {id=WavePlayer_17, properties {soundWaveAssetPtr=/Game/Sounds/Artemis/Biomes/Grasslands/Grassland_Spots/Spot_Grassland_Wind_Rustle_01}}
 {id=WavePlayer_18, properties {soundWaveAssetPtr=/Game/Sounds/Artemis/Biomes/Grasslands/Grassland_Spots/Spot_Grassland_Wind_Rustle_02}}
 {id=WavePlayer_19, properties {soundWaveAssetPtr=/Game/Sounds/Artemis/Biomes/Grasslands/Grassland_Spots/Spot_Grassland_Wind_Rustle_03}}
 {id=WavePlayer_2, properties {soundWaveAssetPtr=&/Spot_Mountain_Day_Gusty_Wind_02}}
 {id=WavePlayer_21, properties {soundWaveAssetPtr=/Game/Sounds/Artemis/Biomes/Grasslands/Grassland_Spots/Spot_Grassland_Wind_Rustle_05}}
 {id=WavePlayer_22, properties {soundWaveAssetPtr=/Game/Sounds/Artemis/Biomes/Grasslands/Grassland_Spots/Spot_Grassland_Wind_Rustle_06}}
 {id=WavePlayer_23, properties {soundWaveAssetPtr=/Game/Sounds/Artemis/Biomes/Grasslands/Grassland_Spots/Spot_Grassland_Wind_Rustle_07}}
 {id=WavePlayer_24, properties {soundWaveAssetPtr=/Game/Sounds/Artemis/Biomes/Grasslands/Grassland_Spots/Spot_Grassland_Wind_Rustle_08}}
 {id=WavePlayer_25, properties {soundWaveAssetPtr=/Game/Sounds/Artemis/Biomes/Grasslands/Grassland_Spots/Spot_Grassland_Wind_Rustle_09}}
 {id=WavePlayer_26, properties {soundWaveAssetPtr=/Game/Sounds/Artemis/Biomes/Grasslands/Grassland_Spots/Spot_Grassland_Wind_Rustle_10}}
 {id=WavePlayer_3, properties {soundWaveAssetPtr=&/Spot_Mountain_Day_Gusty_Wind_03}}
 {id=WavePlayer_4, properties {soundWaveAssetPtr=&/Spot_Mountain_Day_Gusty_Wind_04}}
 {id=WavePlayer_5, properties {soundWaveAssetPtr=&/Spot_Mountain_Day_Gusty_Wind_05}}
 {id=WavePlayer_6, properties {soundWaveAssetPtr=&Spot_Mountain_Day_Gusty_Wind_03}}
 {id=WavePlayer_7, properties {soundWaveAssetPtr=&Spot_Mountain_Day_Gusty_Wind_04}}
 {id=WavePlayer_8, properties {soundWaveAssetPtr=&Spot_Mountain_Day_Gusty_Wind_01}}
 {id=WavePlayer_9, properties {soundWaveAssetPtr=/Game/Sounds/Apollo/Biomes/Alpine/Biome_Alpine_Oneshot_Day_Quad_01}}
]
properties {
 attenuationSettings=/Game/Sounds/Attenuation/Fort_TileAmbient_OneShot_Wind_Att
 bEnableBusSends=false
 maxDistance=8000
 soundClassObject=&Fort_Ambient_OneShots_Outdoors
 volumeMultiplier=0.251
 preEffectBusSends [{}]
}
