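# Sound cue for varmint call ambient one-shot spots in the mountain biome, using bird one-shot attenuation, the outdoor one-shot sound class and an exterior ambient bus send.
# A quality level root feeds the first two tiers from an equally weighted random pick of twelve call waves.
# The last two tiers share a smaller five-wave subset of those calls.
# Each pool passes through its own modulator for slight pitch and volume variation.
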SoundCue /Game/Sounds/Artemis/Biomes/Mountains/Mountain_Spots/Spot_Mountain_Varmint_Call_Cue
root=QualityLevel_0
nodes [
 {id=Modulator_1, children [Random_2], properties {pitchMax=1.1, volumeMax=1.1, volumeMin=0.9}}
 {id=Modulator_2, children [Random_3], properties {pitchMax=1.1, volumeMax=1.1, volumeMin=0.9}}
 {id=QualityLevel_0, children [Modulator_1,Modulator_1,Modulator_2,Modulator_2]}
 {
  id=Random_2
  children [
   WavePlayer_0
   WavePlayer_1
   WavePlayer_2
   WavePlayer_3
   WavePlayer_4
   WavePlayer_13
   WavePlayer_14
   WavePlayer_15
   WavePlayer_16
   WavePlayer_17
   WavePlayer_18
   WavePlayer_19
  ]
  properties {weights [1,1,1,1,1,1,1,1,1,1,1,1]}
 }
 {id=Random_3, children [WavePlayer_20, WavePlayer_21, WavePlayer_22, WavePlayer_23, WavePlayer_24], properties {weights [1,1,1,1,1]}}
 {id=WavePlayer_0, properties {soundWaveAssetPtr=&/Spot_Mountain_Varmint_Call_12}}
 {id=WavePlayer_1, properties {soundWaveAssetPtr=&/Spot_Mountain_Varmint_Call_01}}
 {id=WavePlayer_13, properties {soundWaveAssetPtr=&/Spot_Mountain_Varmint_Call_05}}
 {id=WavePlayer_14, properties {soundWaveAssetPtr=&/Spot_Mountain_Varmint_Call_06}}
 {id=WavePlayer_15, properties {soundWaveAssetPtr=&/Spot_Mountain_Varmint_Call_07}}
 {id=WavePlayer_16, properties {soundWaveAssetPtr=&/Spot_Mountain_Varmint_Call_08}}
 {id=WavePlayer_17, properties {soundWaveAssetPtr=&/Spot_Mountain_Varmint_Call_09}}
 {id=WavePlayer_18, properties {soundWaveAssetPtr=&/Spot_Mountain_Varmint_Call_10}}
 {id=WavePlayer_19, properties {soundWaveAssetPtr=&/Spot_Mountain_Varmint_Call_11}}
 {id=WavePlayer_2, properties {soundWaveAssetPtr=&/Spot_Mountain_Varmint_Call_02}}
 {id=WavePlayer_20, properties {soundWaveAssetPtr=&Spot_Mountain_Varmint_Call_12}}
 {id=WavePlayer_21, properties {soundWaveAssetPtr=&Spot_Mountain_Varmint_Call_08}}
 {id=WavePlayer_22, properties {soundWaveAssetPtr=&Spot_Mountain_Varmint_Call_11}}
 {id=WavePlayer_23, properties {soundWaveAssetPtr=&Spot_Mountain_Varmint_Call_10}}
 {id=WavePlayer_24, properties {soundWaveAssetPtr=&Spot_Mountain_Varmint_Call_09}}
 {id=WavePlayer_3, properties {soundWaveAssetPtr=&/Spot_Mountain_Varmint_Call_03}}
 {id=WavePlayer_4, properties {soundWaveAssetPtr=&/Spot_Mountain_Varmint_Call_04}}
]
properties {
 attenuationSettings=&Fort_TileAmbient_OneShot_Bird_Att
 maxDistance=8005
 soundClassObject=&Fort_Ambient_OneShots_Outdoors
 volumeMultiplier=0.526
 busSends [{audioBus=null, minSendLevel=0.6, sendLevel=1, soundSourceBus=&AmbientAudio_Exterior_SourceBus}]
}
