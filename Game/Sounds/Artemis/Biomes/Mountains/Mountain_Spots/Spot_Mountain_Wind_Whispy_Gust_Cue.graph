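# Sound cue for a one-shot whispy wind gust spot ambience in the Artemis mountain biome, reusing desert gust whistle recordings.
# A quality-level root selects one of two modulated branches, each picking randomly with equal weight from the same five waves.
# The branches differ slightly in volume and pitch variation, with the last two quality slots getting the wider range.
# Outdoor ambient one-shot sound class and tile one-shot attenuation govern playback.
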SoundCue /Game/Sounds/Artemis/Biomes/Mountains/Mountain_Spots/Spot_Mountain_Wind_Whispy_Gust_Cue
root=QualityLevel_0
nodes [
 {id=Modulator_0, children [Random_4], properties {volumeMax=2.1, volumeMin=2}}
 {id=Modulator_2, children [Random_6], properties {pitchMax=1.1, volumeMax=2.2, volumeMin=2}}
 {id=QualityLevel_0, children [Modulator_0,Modulator_0,Modulator_2,Modulator_2]}
 {id=Random_4, children [WavePlayer_18, WavePlayer_19, WavePlayer_20, WavePlayer_21, WavePlayer_22], properties {weights [1,1,1,1,1]}}
 {id=Random_6, children [WavePlayer_28, WavePlayer_29, WavePlayer_30, WavePlayer_31, WavePlayer_32], properties {weights [1,1,1,1,1]}}
 {id=WavePlayer_18, properties {soundWaveAssetPtr=/Game/Sounds/Fort_Ambient_NonLoop/Desert/Wind_DesertGust_Whistle14}}
 {id=WavePlayer_19, properties {soundWaveAssetPtr=/Game/Sounds/Fort_Ambient_NonLoop/Desert/Wind_DesertGust_Whistle15}}
 {id=WavePlayer_20, properties {soundWaveAssetPtr=/Game/Sounds/Fort_Ambient_NonLoop/Desert/Wind_DesertGust_Whistle16}}
 {id=WavePlayer_21, properties {soundWaveAssetPtr=/Game/Sounds/Fort_Ambient_NonLoop/Desert/Wind_DesertGust_Whistle18}}
 {id=WavePlayer_22, properties {soundWaveAssetPtr=/Game/Sounds/Fort_Ambient_NonLoop/Desert/Wind_DesertGust_Whistle26}}
 {id=WavePlayer_28, properties {soundWaveAssetPtr=&Wind_DesertGust_Whistle14}}
 {id=WavePlayer_29, properties {soundWaveAssetPtr=&Wind_DesertGust_Whistle15}}
 {id=WavePlayer_30, properties {soundWaveAssetPtr=&Wind_DesertGust_Whistle16}}
 {id=WavePlayer_31, properties {soundWaveAssetPtr=&Wind_DesertGust_Whistle18}}
 {id=WavePlayer_32, properties {soundWaveAssetPtr=&Wind_DesertGust_Whistle26}}
]
properties {
 attenuationSettings=/Game/Sounds/Attenuation/Fort_TileAmbient_OneShot_Att
 bEnableBusSends=false
 maxDistance=8000
 soundClassObject=&Fort_Ambient_OneShots_Outdoors
 volumeMultiplier=0.551
 preEffectBusSends [{}]
}
